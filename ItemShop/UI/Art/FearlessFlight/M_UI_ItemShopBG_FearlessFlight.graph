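# UI material for a FearlessFlight Item Shop background: a space-web image with blue and magenta glow.
# An aspect-corrected web texture, warped by panning Gaussian noise, is tinted magenta by that noise and screen-blended with a blue horizontal center band.
# A vertical falloff mask fades the result to black toward top and bottom.
# The graph also holds particle-grid and center-scaled web branches, with no material output connections shown.
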Material /ItemShop/UI/Art/FearlessFlight/M_UI_ItemShopBG_FearlessFlight
graph {
 properties {bUsedWithStaticMesh=false, materialDomain=MD_UI, pixelDepthOffsetMode=PDOM_Legacy, usedShadingModels=Default Lit}
 nodes [
  Constant3Vector() {constant=<rgba 0 0.0205 0.8177 1>}
  n1 = MaterialFunctionCall() {materialFunction=&GetUserInterfaceUV}
  n2 = ComponentMask(None=Abs(None=ConstantBiasScale(None=@n1.Normalized UV) {bias=-0.5, scale=2})) {g=true}
  n3 = ScalarParameter() {defaultValue=0.05, parameterName=SubtractBlue}
  n4 = Saturate(None=Multiply(A=Subtract(A=@n2, B=@n3), B=ScalarParameter() {defaultValue=2.25, parameterName=FalloffBlue}))
  n5 = Multiply(A=Multiply(A=OneMinus(None=@n4), B=Constant3Vector() {constant=<rgba 0 0.1304 1 1>})) {constB=0.25}
  n6 = MaterialFunctionCall(Float2=@n1.Pixel Size) {materialFunction=&BreakOutFloat2Components}
  n7 = Divide(A=@n6.R, B=Max(A=@n6.G))
  n8 = AppendVector(A=@n7, B=Constant() {r=1})
  n9 = Multiply(A=OneMinus(None=Multiply(B=@n8) {constA=1.125, constB=1.125}))
  n10 = Multiply(A=@n8, B=@n1.Normalized UV)
  n11 = Multiply(A=@n7, B=@n1.Normalized UV)
  n12 = Time()
  n13 = TextureSample(UVs=Panner(Coordinate=Multiply(A=@n11) {constB=0.35}, Time=Multiply(A=@n12) {constB=0.75}) {speedX=0.025}) {texture=/Game/Effects/Fort_Effects/Textures/Tile/T_GaussianNoise_01_M}
  n14 = TextureSample(UVs=Add(A=Add(A=@n9, B=@n10), B=Multiply(A=@n13.R) {constB=0.0125})) {texture=&/T_UI_WebSpace}
  n15 = Multiply(A=@n13.R) {constB=0.15}
  n16 = Power(Base=@n15) {constExponent=0.125}
  n17 = LinearInterpolate(
   A=Multiply(A=Multiply(A=@n14.RGB) {constB=1.25}, B=OneMinus(None=@n16))
   B=Constant3Vector() {constant=<rgba 0.5452 0.0135 0.6445 1>}
   Alpha=@n15
  )
  n18 = Abs(None=ConstantBiasScale(None=MaterialFunctionCall() {materialFunction=&MF_UI_UVScaling}.Scaled To Fill) {bias=-0.5, scale=2})
  n19 = Multiply(
   A=Subtract(A=ComponentMask(None=@n18) {g=true}, B=ScalarParameter() {defaultValue=0.37, parameterName=Subtract})
   B=ScalarParameter() {defaultValue=3.5, parameterName=Falloff}
  )
  LinearInterpolate(
   A=Add(
    A=MaterialFunctionCall!(Blend=@n5, Base=@n17) {materialFunction=/Engine/Functions/Engine_MaterialFunctions03/Blends/Blend_Screen}.Result
    B=@n17
   )
   Alpha=Multiply(A=Saturate(None=@n19), B=ScalarParameter() {defaultValue=1.5, parameterName=Intensity})
  ) {constB=0}
  n20 = MaterialFunctionCall!(
   UV=MaterialFunctionCall() {materialFunction=&MF_UI_UVScaling}.Scaled To Fill
   Global Offset
   Global Rotation=Constant() {r=0.35}
   Tiling=MaterialFunctionCall(X=Constant() {r=2}, Y=Constant() {r=2}) {materialFunction=&MakeFloat2}.Result
   Cell Offset=Add(A=Constant() {r=5}) {constB=2}
   Cell Scale=Constant() {r=1}
   Cell Rotation=Add(B=NamedRerouteUsage()) {constA=1}
   Random Seed
  ) {materialFunction=/FortUILibrary/MaterialFunctions/Basics/MF_UI_ParticleGrid}
  n21 = MaterialFunctionCall() {materialFunction=&MF_UI_UVScaling}
  n22 = Abs(None=ConstantBiasScale(None=@n21.Scaled To Fill) {bias=-0.5, scale=2})
  n23 = Multiply(
   A=Subtract(A=ComponentMask(None=@n22) {g=true}, B=ScalarParameter() {defaultValue=0.37, parameterName=Subtract})
   B=ScalarParameter() {defaultValue=3.5, parameterName=Falloff}
  )
  LinearInterpolate(
   A=Multiply(A=TextureSample(UVs=@n20.Cells UV) {texture=&T_UI_WebSpace}.RGB) {constB=1.5}
   Alpha=Multiply(A=Saturate(None=@n23), B=ScalarParameter() {defaultValue=1.5, parameterName=Intensity})
  ) {constB=0}
  n24 = Add(A=@n21.Scaled To Fill) {constB=0.0001}
  n25 = Constant() {r=0.5}
  Multiply(A=TextureSample(UVs=MaterialFunctionCall(UVs=@n24, Texture Scale=@n25) {materialFunction=&ScaleUVsByCenter}.UVs) {texture=&T_UI_WebSpace}.R) {constB=5}
  Multiply(A=@n15) {constB=0.75}
  NamedRerouteDeclaration(None=@n20.Random Cell Values) {name=RandomCellValues, nodeColor=<rgba 0 1 0.8594 1>}
  NamedRerouteDeclaration(None=@n20.Random Rows) {name=RandomRows, nodeColor=<rgba 0 1 0.8594 1>}
  NamedRerouteDeclaration(None=@n20.Random Columns) {name=RandomColumns, nodeColor=<rgba 0 1 0.8594 1>}
  ScalarParameter() {defaultValue=1, parameterName=IntensityBlue}
 ]
}
