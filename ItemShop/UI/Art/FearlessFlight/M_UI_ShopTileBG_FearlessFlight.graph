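# UI material for the Fearless Flight item shop tile background, an animated hexagonal portal over a starburst backdrop.
# Named reroutes organize a purple gradient, CRT lines distorted by a flow map, starburst and radiating-line textures, and streaks panned in radial coordinates.
# Four portal haze layers use scaled, time-rotated UVs and add MF_UI_Hexagon runner patterns.
# Their combined alpha masks a blend of these layers with a three-color radial gradient over the background.
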Material /ItemShop/UI/Art/FearlessFlight/M_UI_ShopTileBG_FearlessFlight
graph {
 properties {bUsedWithStaticMesh=false, materialDomain=MD_UI, pixelDepthOffsetMode=PDOM_Legacy, usedShadingModels=Default Lit}
 nodes [
  ComponentMask() {g=true, r=true}
  n1 = Constant3Vector() {constant=<rgba 1 0.5128 0.0085 1>}
  n2 = Constant3Vector() {constant=<rgba 1 0.1592 0 1>}
  n3 = Constant3Vector() {constant=<rgba 1 0.0079 0 1>}
  n4 = MaterialFunctionCall(UV=NamedRerouteUsage(), GradientStart=Constant(), GradientEnd=Constant() {r=1}) {materialFunction=&MF_UI_RadialGradient}
  n5 = TextureSample(UVs=NamedRerouteUsage(), Tex=NamedRerouteUsage())
  n6 = NamedRerouteUsage()
  n7 = TextureSample(UVs=@n6, Tex=NamedRerouteUsage())
  n8 = NamedRerouteUsage()
  n9 = TextureSample(UVs=@n8, Tex=NamedRerouteUsage())
  n10 = NamedRerouteUsage()
  n11 = TextureSample(UVs=@n10, Tex=NamedRerouteUsage())
  n12 = Saturate(None=Add(A=Add(A=@n5.A, B=@n7.A), B=Add(A=@n9.A, B=@n11.A)))
  n13 = Constant() {r=0.005}
  n14 = Add(
   A=Multiply(
    A=MaterialFunctionCall!(time, UV=@n8, time_gain, UV_Static, hexTimeGain, PatternHexIntensity=@n13, flip) {materialFunction=/OfferCatalog/Art/_MaterialFunctions_ShopTheming/MF_UI_Hexagon}.Result
    B=NamedRerouteUsage()
   )
   B=@n9.RGB
  )
  n15 = MaterialFunctionCall(UV=@n10, PatternHexIntensity=@n13) {materialFunction=&MF_UI_Hexagon}
  n16 = Add(A=Multiply(A=Multiply(A=@n15.Result) {constB=5}, B=NamedRerouteUsage()), B=@n11.RGB)
  n17 = MaterialFunctionCall(UV=@n6, PatternHexIntensity=@n13) {materialFunction=&MF_UI_Hexagon}
  n18 = Add(A=Multiply(A=Multiply(A=@n17.Result) {constB=3}, B=NamedRerouteUsage()), B=Multiply(A=@n7.RGB) {constB=0.65})
  n19 = MaterialFunctionCall(UV=NamedRerouteUsage(), PatternHexIntensity=@n13) {materialFunction=&MF_UI_Hexagon}
  n20 = Add(A=Multiply(A=NamedRerouteUsage(), B=Multiply(A=@n19.Result) {constB=5}), B=Multiply(A=@n5.RGB) {constB=0.8})
  n21 = Multiply(A=Add(A=Add(A=@n18, B=@n20), B=Add(A=@n14, B=@n16))) {constB=3}
  n22 = NamedRerouteUsage()
  LinearInterpolate(
   A=NamedRerouteUsage()
   B=Max(
    A=MaterialFunctionCall(A=@n1, B=@n2, C=@n3, Alpha=@n4.Smooth) {materialFunction=&3ColorBlend}.Result
    B=Multiply(A=@n12, B=@n21)
   )
   Alpha=Subtract(A=@n12, B=Multiply(A=@n22) {constB=2})
  )
  LinearInterpolate(A=@n1, B=@n2, Alpha=@n4.Smooth)
  MaterialFunctionCall() {materialFunction=&MF_UI_UVScaling}
  MaterialFunctionCall(
   UVs=MaterialFunctionCall!(ScreenUV, TextureSize, ScalePivotPoint) {materialFunction=/FortUILibrary/MaterialFunctions/Basics/MF_UI_ScreenSpaceUVScaling}.Scaled to Fill
   Texture Scale=Constant() {r=0.5}
  ) {materialFunction=&ScaleUVsByCenter}
  NamedRerouteDeclaration(None=TextureCoordinate()) {name=ScreenUVs, nodeColor=<rgba 1 0.867 0 1>}
  n23 = NamedRerouteUsage()
  n24 = Time()
  NamedRerouteDeclaration(None=Rotator(Coordinate=@n23, Time=Multiply(A=@n24) {constB=0.496})) {name=Spins, nodeColor=<rgba 1 0 0.375 1>}
  n25 = Constant3Vector() {constant=<rgba 0.1413 0.0027 0.2016 1>}
  n26 = Constant3Vector() {constant=<rgba 0.4735 0.0006 0.5647 1>}
  n27 = MaterialFunctionCall(UV=NamedRerouteUsage(), GradientStart=Constant(), GradientEnd=Constant() {r=1}) {materialFunction=&MF_UI_RadialGradient}
  n28 = NamedRerouteUsage()
  n29 = ComponentMask(None=Abs(None=ConstantBiasScale(None=@n28) {bias=-0.5, scale=2})) {g=true}
  n30 = ScalarParameter() {defaultValue=0.25, parameterName=Subtract}
  n31 = Saturate(None=Multiply(A=Subtract(A=@n29, B=@n30), B=ScalarParameter() {defaultValue=3, parameterName=Falloff}))
  n32 = Constant3Vector() {constant=<rgba 0.0648 0.0006 0.3372 1>}
  NamedRerouteDeclaration(None=LinearInterpolate(A=LinearInterpolate(A=@n25, B=@n26, Alpha=@n27.Smooth), B=Multiply(A=@n31, B=@n32), Alpha=@n31)) {name=GradientBG, nodeColor=<rgba 1 0 0.2109 1>}
  n33 = Add(A=Multiply(A=NamedRerouteUsage()) {constB=0.745}, B=NamedRerouteUsage())
  n34 = Multiply(A=NamedRerouteUsage()) {constB=0.25}
  NamedRerouteDeclaration(None=Add(
    A=@n33
    B=MaterialFunctionCall!(Base=@n33, Blend=@n34) {materialFunction=/Engine/Functions/Engine_MaterialFunctions03/Blends/Blend_HardLight}.Result
   )) {name=FinalBG, nodeColor=<rgba 0.6846 0 1 1>}
  n35 = NamedRerouteUsage()
  n36 = Constant2Vector() {g=1.25, r=1}
  n37 = TextureSample(UVs=MaterialFunctionCall(UVs=Multiply(A=@n35, B=@n36), Texture Scale=Constant() {r=0.8}) {materialFunction=&ScaleUVsByCenter}.UVs) {samplerType=SAMPLERTYPE_LinearColor, texture=/OfferCatalog/Art/FearlessFlight/T_UI_Starburst}
  n38 = Constant3Vector() {constant=<rgba 0 0.1293 1 1>}
  n39 = Multiply(A=@n37.G) {constB=5}
  n40 = Constant3Vector() {constant=<rgba 0 0.1449 0.5226 1>}
  n41 = Add(
   A=Add(A=Multiply(A=@n37.R, B=@n38), B=Multiply(A=@n39, B=@n40))
   B=Multiply(A=@n37.B, B=Constant3Vector() {constant=<rgba 0 0.0806 0.5694 1>})
  )
  n42 = Constant2Vector() {g=0.85, r=1}
  n43 = TextureSample(
   UVs=MaterialFunctionCall(UVs=Multiply(A=@n35, B=@n42), Texture Scale=Constant() {r=0.5}) {materialFunction=&ScaleUVsByCenter}.UVs
   Tex=TextureObject() {samplerType=SAMPLERTYPE_Alpha, texture=/OfferCatalog/Art/FearlessFlight/T_UI_RadiatingLines}
  ) {samplerType=SAMPLERTYPE_LinearColor, texture=&T_UI_RadiatingLines}
  n44 = Multiply(A=Multiply(A=Multiply(A=@n43.R) {constB=1.25}, B=Constant3Vector() {constant=<rgba 1 0.9962 0 1>})) {constB=0.5}
  NamedRerouteDeclaration(None=Add(A=NamedRerouteUsage(), B=Add(A=@n41, B=@n44))) {name=StarburstBG, nodeColor=<rgba 0 0.9766 1 1>}
  NamedRerouteDeclaration(None=@n43.R) {name=BurstAlpha, nodeColor=<rgba 0 0.2266 1 1>}
  NamedRerouteDeclaration(None=Rotator(Coordinate=@n23, Time=Multiply(A=@n24) {constB=0.52})) {name=OuterHexSpins, nodeColor=<rgba 1 0 0.375 1>}
  NamedRerouteDeclaration(None=MaterialFunctionCall(UVs=NamedRerouteUsage(), Texture Scale=Constant() {r=1.7}) {materialFunction=&ScaleUVsByCenter}.UVs) {name=OuterMostHexUVs, nodeColor=<rgba 1 0.7734 0 1>}
  n45 = NamedRerouteUsage()
  n46 = Constant3Vector() {constant=<rgba 1 0.5501 0 0>}
  n47 = MaterialFunctionCall(Vector or UVs=Subtract(A=NamedRerouteUsage()) {constB=0.5}) {materialFunction=&VectorToRadialValue}
  n48 = Panner(
   Coordinate=@n47.Radial Coordinates
   Speed=AppendVector(A=ScalarParameter() {defaultValue=0.0125, parameterName=U-Speed}, B=ScalarParameter() {defaultValue=-0.5, parameterName=V-Speed})
  )
  n49 = Constant() {r=1}
  n50 = ScalarParameter() {defaultValue=0.009, parameterName=R-V-Tiles}
  n51 = TextureSample(UVs=Multiply(A=@n48, B=AppendVector(A=@n49, B=@n50))) {texture=/Game/Global/Textures/Default/PointSplatter/T_Ui_ColumbusSeries_512}
  n52 = ScalarParameter() {defaultValue=3, parameterName=G-U-Tiles}
  n53 = ScalarParameter() {defaultValue=0.03, parameterName=G-V-Tiles}
  n54 = TextureSample(UVs=Multiply(A=@n48, B=AppendVector(A=@n52, B=@n53))) {texture=&T_Ui_ColumbusSeries_512}
  n55 = Multiply(
   A=Power(Base=@n47.Linear Distance, Exp=ScalarParameter() {defaultValue=1, parameterName=MaskPower})
   B=Add(A=Multiply(A=@n51.B) {constB=2}, B=Multiply(A=@n54.B) {constB=2})
  )
  NamedRerouteDeclaration(None=Add(
    A=LinearInterpolate(A=@n45, B=@n46, Alpha=@n55)
    B=MaterialFunctionCall(Base=@n55, Blend=@n46) {materialFunction=&Blend_Overlay}.Result
   )) {name=Streaks, nodeColor=<rgba 0 0.2031 1 1>}
  NamedRerouteDeclaration(None=@n55) {name=StreaksAlpha, nodeColor=<rgba 0 1 0.5078 1>}
  NamedRerouteDeclaration(None=@n37.B) {name=StarburstAlpha, nodeColor=<rgba 1 0 0.9844 1>}
  n56 = Multiply(A=NamedRerouteUsage(), B=Constant2Vector() {g=200, r=84}) {constB=75}
  n57 = MaterialFunctionCall!(
   Texture 2D=TextureObject() {texture=/OfferCatalog/Art/Sunburst/T_UI_CRTTexture}
   Texture UVs=MaterialFunctionCall(UVs=Frac(None=@n56), Texture Scale=Constant2Vector() {g=0.5, r=2}) {materialFunction=&ScaleUVsByCenter}.UVs
   FlowMap Texture 2D=TextureObject() {texture=/OfferCatalog/Art/Sunburst/T_UI_CirclesFlowmap}
   FlowMap UVs
   FlowIntensity=Constant() {r=0.75}
   FlowSpeed=Constant() {r=0.484}
  ) {materialFunction=/FortUILibrary/MaterialFunctions/Effects/MF_UI_FlowMap}
  n58 = Max(A=Max(A=ComponentMask(None=@n57.Result) {r=true}, B=ComponentMask(None=@n57.Result) {g=true}), B=ComponentMask(None=@n57.Result) {b=true})
  NamedRerouteDeclaration(None=LinearInterpolate(
    A=Constant3Vector() {constant=<rgba 0.0395 0.0009 0.0343 1>}
    B=Constant3Vector() {constant=<rgba 0.1221 0 0.1144 1>}
    Alpha=Multiply(A=@n58) {constB=1.25}
   )) {name=CRT Lines, nodeColor=<rgba 1 0 0.8672 1>}
  NamedRerouteDeclaration(None=TextureObjectParameter() {parameterName=Tex, texture=/OfferCatalog/Art/FearlessFlight/T_UI_PortalHaze}) {name=HexTexture, nodeColor=<rgba 1 0.2403 0 1>}
  NamedRerouteDeclaration(None=TextureObject() {samplerType=SAMPLERTYPE_Alpha, texture=/OfferCatalog/Art/FearlessFlight/T_UI_PortalMask}) {name=HexMask, nodeColor=<rgba 1 1 1 1>}
  NamedRerouteDeclaration(None=Rotator(Coordinate=@n23, Time=Multiply(A=@n24) {constB=0.493})) {name=MedHexRotation, nodeColor=<rgba 1 0 0.4688 1>}
  NamedRerouteDeclaration(None=MaterialFunctionCall(UVs=NamedRerouteUsage(), Texture Scale=Constant() {r=0.35}) {materialFunction=&ScaleUVsByCenter}.UVs) {name=SmallHexUVs, nodeColor=<rgba 1 0.7734 0 1>}
  NamedRerouteDeclaration(None=MaterialFunctionCall(UVs=NamedRerouteUsage(), Texture Scale=Constant() {r=0.6}) {materialFunction=&ScaleUVsByCenter}.UVs) {name=MediumHexUVs, nodeColor=<rgba 0.625 0.4154 0 1>}
  NamedRerouteDeclaration(None=Rotator(Coordinate=@n23, Time=Multiply(A=@n24) {constB=0.49})) {name=SmallHexRotation, nodeColor=<rgba 0 0.8828 1 1>}
  NamedRerouteDeclaration(None=Constant3Vector() {constant=<rgba 1 0.6657 0 1>}) {name=HexRunnerColor, nodeColor=<rgba 0 0.9531 1 1>}
  NamedRerouteUsage()
  ScalarParameter() {defaultValue=2, parameterName=Intensity}
 ]
}
